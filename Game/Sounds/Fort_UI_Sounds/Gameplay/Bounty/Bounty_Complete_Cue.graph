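# Sound cue for the UI sound played, by its name, when a bounty completes in gameplay.
# A single wave player plays the Quest_Squad_Chest_01 wave at reduced volume, with no randomization or mixing.
# It routes to the critical in-game UI sound class and references the Discovery POI concurrency rule.
# Its stored single-instance, prevent-new concurrency override is not switched on, so that shared rule governs overlapping plays.
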SoundCue /Game/Sounds/Fort_UI_Sounds/Gameplay/Bounty/Bounty_Complete_Cue
root=WavePlayer_0
nodes [{id=WavePlayer_0, properties {soundWaveAssetPtr=&/Quest_Squad_Chest_01}}]
properties {
 maxDistance=2097152
 soundClassObject=&Fort_UI_InGame_Critical
 volumeMultiplier=0.6
 concurrencyOverrides {maxCount=1, resolutionRule=PreventNew}
 concurrencySet [/Game/Sounds/Concurrency_Rules/CONC_Discovery_POI]
}
