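# Sound cue named for the time-up moment of Bounty gameplay, assigned to the critical in-game UI sound class.
# It has no nodes and no root, so the file references no audio to play.
# It keeps a reduced volume and a concurrency override that allows one instance per owner, blocks new plays while one is active, and sets a retrigger time.
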SoundCue /Game/Sounds/Fort_UI_Sounds/Gameplay/Bounty/Bounty_TimeUp_Cue
root=null
nodes []
properties {
 bOverrideConcurrency=true
 maxDistance=2097152
 soundClassObject=&Fort_UI_InGame_Critical
 volumeMultiplier=0.6
 concurrencyOverrides {bLimitToOwner=true, maxCount=1, resolutionRule=PreventNew, retriggerTime=10}
}
