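# Sound cue for the tonal stinger of an accepted urgent quest, filed under the Bounty gameplay UI sounds.
# A single wave player feeds the output directly at slightly reduced volume, routed through the in-game SFX sound class.
# A concurrency override allows one instance at a time, stopping the oldest when a new one starts.
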SoundCue /Game/Sounds/Fort_UI_Sounds/Gameplay/Bounty/Urgent_Quest_Accepted_Cue
root=WavePlayer_3
nodes [{id=WavePlayer_3, properties {soundWaveAssetPtr=&/Urgent_Quest_Stinger_Appear_Tonal_01}}]
properties {
 bOverrideConcurrency=true
 maxDistance=2097152
 soundClassObject=&Fort_SFX_InGame
 volumeMultiplier=0.8
 concurrencyOverrides {maxCount=1, resolutionRule=StopOldest}
}
