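# Sound cue for the sting that, going by its name, plays when an urgent Bootcamp quest is accepted.
# A single wave player feeds the Bootcamp checklist sting straight to the output at slightly reduced volume, routed through the in-game SFX sound class.
# A concurrency override allows one instance at a time, stopping the oldest when a new one starts.
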SoundCue /Game/Sounds/Fort_UI_Sounds/Gameplay/Bounty/Urgent_Quest_Bootcamp_Accepted_Cue
root=WavePlayer_1
nodes [{id=WavePlayer_1, properties {soundWaveAssetPtr=&/Sting_Quest_BootCamp_Checklist}}]
properties {
 bOverrideConcurrency=true
 maxDistance=2097152
 soundClassObject=&Fort_SFX_InGame
 volumeMultiplier=0.8
 concurrencyOverrides {maxCount=1, resolutionRule=StopOldest}
}
