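# Sound cue for the urgent exploration quest accepted stinger, filed under the gameplay bounty UI sounds.
# A Mixer layers two wave players at equal volume: the exploration stinger and the urgent quest appear SFX.
# It routes through the Fort_SFX_InGame sound class and overrides concurrency to a single instance, so a new play stops the older one.
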SoundCue /Game/Sounds/Fort_UI_Sounds/Gameplay/Bounty/Urgent_Quest_Exploration_Accepted_Cue
root=Mixer_1
nodes [
 {id=Mixer_1, children [WavePlayer_5,WavePlayer_4], properties {inputVolume [1,1]}}
 {id=WavePlayer_4, properties {soundWaveAssetPtr=&/Urgent_Quest_Exploration_Stinger}}
 {id=WavePlayer_5, properties {soundWaveAssetPtr=&/Urgent_Quest_Stinger_SFX_Appear}}
]
properties {
 bOverrideConcurrency=true
 maxDistance=2097152
 soundClassObject=&Fort_SFX_InGame
 volumeMultiplier=1
 concurrencyOverrides {maxCount=1, resolutionRule=StopOldest}
}
