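# Sound cue for the Season 21 DBZ Duelist quest stinger, filed with the gameplay bounty UI sounds.
# A single wave player plays the stinger wave at reduced volume through a 2D hit-notification instigator sound class.
# It has high priority and uses the shared Discovery POI concurrency rule, with stored overrides for one instance that stops the oldest.
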SoundCue /Game/Sounds/Fort_UI_Sounds/Gameplay/Bounty/s21_Quest_Duelist_DBZ
root=WavePlayer_0
nodes [{id=WavePlayer_0, properties {soundWaveAssetPtr=&/s21_DBZ_S_Quest_Duelist_Stinger}}]
properties {
 maxDistance=2097152
 priority=0.9
 soundClassObject=&Fort_2DHitNotifications_Instigate_PSM
 volumeMultiplier=0.3
 concurrencyOverrides {maxCount=1, resolutionRule=StopOldest}
 concurrencySet [/Game/Sounds/Concurrency_Rules/CONC_Discovery_POI]
}
